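# Idle cosmetic Niagara effect for the Arctic Fox Warrior skin, adding slowly spinning mesh shards and frosty mist to the character.
# Four shard emitters (Back, Middle, Front, Extra) each burst two persistent, randomly sized meshes, pinned by index with offsets to User.SkeletalMesh bones or sockets.
# Rounded MaterialRandom picks each shard's spin direction.
# FrostyMist_Local continuously spawns pale-blue flipbook sprites alternating between two attach points, sinking as they grow and fade in and out.
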NiagaraSystem /BRCosmetics/Characters/Player/Female/Medium/Bodies/F_MED_Fox_Warrior_Arctic/Effects/NS_Skin_FoxWarrior_Arctic_Idle
dependencies {usedDataInterfaces [&NiagaraDataInterfaceCurve], usedDynamicInputs [], usedModules [&SystemState]}
emitters {
 BackShard {
  emitterUpdateScript [
   &EmitterState()
   &SpawnBurst_Instantaneous(
    Spawn Count=2
    Spawn Probability=&MakeCustomFloatFromBool(
     Bool=&SetBoolByIntComparison(A=$Emitter.LoopCount, B=0, Comparison Type=::NewEnumerator4)
     True Float=1
     False Float=0
    )
    Use Spawn Probability=true
   )
  ]
  particleSpawnScript [
   /SetVariables(Particles.ShardSize:NiagaraFloat=&Multiply_Float(
     A=&RandomRangeFloat(Minimum=0.65, Maximum=1)
     B=&RandomRangeFloat(Minimum=0.06, Maximum=0.11)
    )) {isSetParameters=true, module=SetVariables_#1}
   &InitializeParticle(
    Mesh Scale=&MakeVector(
     X=$Particles.ShardSize
     Y=$Particles.ShardSize
     Z=&Add_Float(A=-0.0625, B=&Multiply_Float(A=1.875, B=$Particles.ShardSize))
    )
    Mesh Scale Mode=::NewEnumerator4
    UsePositionOffset=false
    Write Material Random=true
   ) {module=InitializeParticle001}
   &InitialMeshOrientation(
    Orientation Axis=<0 0 1>
    Orientation Vector=<0 0 1>
    Rotation=&RandomRangeVector(Minimum=<-0.05 0.45 -0.05>, Maximum=<0.05 0.55 0.05>)
    Use Orientation Vector=true
    Use Rotation Vector=true
   )
   /SetVariables() {isSetParameters=true}
  ]
  particleUpdateScript [
   &ParticleState(
    Kill Particles When Lifetime Has Elapsed=false
    Let Infinitely Lived Particles Die When Emitter Deactivates=true
    Loop Particles Lifetime=true
   )
   &SkeletalMeshLocation(
    Bone / Socket Index=&ReturnExecIndex()
    Inherit Velocity=false
    Interpolate Mesh Sampling=false
    MaskSpawnGroup=false
    Mesh Sampling Type=::NewEnumerator5
    Sampled Position Offset=&Lerp_Vector(A=<3 -8 2>, B=<3 8 2>, Alpha=&MakeFloatFromInt(Int=&ReturnExecIndex()))
    Sampled Position Offset Coordinate Space:ENiagaraMeshTransforms=::NewEnumerator3
    Skeletal Mesh=$User.SkeletalMesh
    Socket Sampling Mode=::NewEnumerator5
   )
   &UpdateMeshOrientation(
    Rotation Coordinate Space=::NewEnumerator3
    Rotation Rate=&UniformAOrBFloat(
     A=&RandomRangeFloat(Minimum=-0.2, Maximum=-0.5)
     B=&RandomRangeFloat(Minimum=0.2, Maximum=0.5)
     Distribution Weight=&ClampFloat(Float=/Niagara/DynamicInputs/Integer/V2/Round_Float(Float:NiagaraFloat=$Particles.Initial.MaterialRandom))
    )
    Rotation Vector=<0 0 1>
   )
  ]
  rendererClasses [&NiagaraMeshRendererProperties]
  renderers [{rendererClass=&NiagaraMeshRendererProperties, rendererIndex=0}]
 }
 ExtraShard {
  emitterUpdateScript [
   &EmitterState()
   &SpawnBurst_Instantaneous(
    Spawn Count=2
    Spawn Probability=&MakeCustomFloatFromBool(
     Bool=&SetBoolByIntComparison(A=$Emitter.LoopCount, B=0, Comparison Type=::NewEnumerator4)
     True Float=1
     False Float=0
    )
    Use Spawn Probability=true
   )
  ]
  particleSpawnScript [
   /SetVariables(Particles.ShardSize:NiagaraFloat=&Multiply_Float(
     A=&RandomRangeFloat(Minimum=0.65, Maximum=1)
     B=&RandomRangeFloat(Minimum=0.06, Maximum=0.11)
    )) {isSetParameters=true, module=SetVariables_#1}
   &InitializeParticle(
    Mesh Scale=&MakeVector(
     X=$Particles.ShardSize
     Y=$Particles.ShardSize
     Z=&Add_Float(A=-0.0625, B=&Multiply_Float(A=1.875, B=$Particles.ShardSize))
    )
    Mesh Scale Mode=::NewEnumerator4
    UsePositionOffset=false
    Write Material Random=true
   ) {module=InitializeParticle001}
   &InitialMeshOrientation(
    Orientation Axis=<0 0 1>
    Orientation Vector=<0 0 1>
    Rotation=&RandomRangeVector(Minimum=<-0.05 0.43 -0.05>, Maximum=<0.05 0.47 0.05>)
    Use Orientation Vector=true
    Use Rotation Vector=true
   )
  ]
  particleUpdateScript [
   &ParticleState(
    Kill Particles When Lifetime Has Elapsed=false
    Let Infinitely Lived Particles Die When Emitter Deactivates=true
    Loop Particles Lifetime=true
   )
   &SkeletalMeshLocation(
    Bone / Socket Index=&ReturnExecIndex()
    Inherit Velocity=false
    Interpolate Mesh Sampling=false
    MaskSpawnGroup=false
    Mesh Sampling Type=::NewEnumerator5
    Sampled Position Offset=&Lerp_Vector(
     A=<8 -1 -3>
     B=<8 1 -3>
     Alpha=&MakeFloatFromInt(Int=&ReturnExecIndex())
    )
    Sampled Position Offset Coordinate Space:ENiagaraMeshTransforms=::NewEnumerator3
    Skeletal Mesh=$User.SkeletalMesh
    Socket Sampling Mode=::NewEnumerator5
   )
   &UpdateMeshOrientation(
    Rotation Coordinate Space=::NewEnumerator3
    Rotation Rate=&UniformAOrBFloat(
     A=&RandomRangeFloat(Minimum=-0.2, Maximum=-0.5)
     B=&RandomRangeFloat(Minimum=0.2, Maximum=0.5)
     Distribution Weight=&ClampFloat(Float=&Round_Float(Float:NiagaraFloat=$Particles.Initial.MaterialRandom))
    )
    Rotation Vector=<0 0 1>
   )
  ]
  rendererClasses [&NiagaraMeshRendererProperties]
  renderers [{rendererClass=&NiagaraMeshRendererProperties, rendererIndex=0}]
 }
 FrontShard {
  emitterUpdateScript [
   &EmitterState()
   &SpawnBurst_Instantaneous(
    Spawn Count=2
    Spawn Probability=&MakeCustomFloatFromBool(
     Bool=&SetBoolByIntComparison(A=$Emitter.LoopCount, B=0, Comparison Type=::NewEnumerator4)
     True Float=1
     False Float=0
    )
    Use Spawn Probability=true
   )
  ]
  particleSpawnScript [
   /SetVariables(Particles.ShardSize:NiagaraFloat=&Multiply_Float(
     A=&RandomRangeFloat(Minimum=0.65, Maximum=1)
     B=&RandomRangeFloat(Minimum=0.06, Maximum=0.11)
    )) {isSetParameters=true, module=SetVariables_#1}
   &InitializeParticle(
    Mesh Scale=&MakeVector(
     X=$Particles.ShardSize
     Y=$Particles.ShardSize
     Z=&Add_Float(A=-0.0625, B=&Multiply_Float(A=1.875, B=$Particles.ShardSize))
    )
    Mesh Scale Mode=::NewEnumerator4
    UsePositionOffset=false
    Write Material Random=true
   ) {module=InitializeParticle001}
   &InitialMeshOrientation(
    Orientation Axis=<0 0 1>
    Orientation Vector=<0 0 1>
    Rotation=&RandomRangeVector(Minimum=<-0.05 0.45 -0.05>, Maximum=<0.05 0.55 0.05>)
    Use Orientation Vector=true
    Use Rotation Vector=true
   )
  ]
  particleUpdateScript [
   &ParticleState(
    Kill Particles When Lifetime Has Elapsed=false
    Let Infinitely Lived Particles Die When Emitter Deactivates=true
    Loop Particles Lifetime=true
   )
   &SkeletalMeshLocation(
    Bone / Socket Index=&ReturnExecIndex()
    Inherit Velocity=false
    Interpolate Mesh Sampling=false
    MaskSpawnGroup=false
    Mesh Sampling Type=::NewEnumerator5
    Sampled Position Offset=&Lerp_Vector(A=<5 3 2>, B=<5 -3 2>, Alpha=&MakeFloatFromInt(Int=&ReturnExecIndex()))
    Sampled Position Offset Coordinate Space:ENiagaraMeshTransforms=::NewEnumerator3
    Skeletal Mesh=$User.SkeletalMesh
    Socket Sampling Mode=::NewEnumerator5
   )
   &UpdateMeshOrientation(
    Rotation Coordinate Space=::NewEnumerator3
    Rotation Rate=&UniformAOrBFloat(
     A=&RandomRangeFloat(Minimum=-0.1, Maximum=-0.5)
     B=&RandomRangeFloat(Minimum=0.1, Maximum=0.5)
     Distribution Weight=&ClampFloat(Float=&Round_Float(Float:NiagaraFloat=$Particles.Initial.MaterialRandom))
    )
    Rotation Vector=<0 0 1>
   )
  ]
  rendererClasses [&NiagaraMeshRendererProperties]
  renderers [{rendererClass=&NiagaraMeshRendererProperties, rendererIndex=0}]
 }
 FrostyMist_Local {
  emitterUpdateScript [
   &EmitterState()
   &SpawnRate(SpawnRate=9, Use Spawn Probability=false)
  ]
  particleSpawnScript [
   /Niagara/Modules/Spawn/Initialization/InitializeParticle(
    Color=<rgba 0.35 0.5992 1 0.125>
    Lifetime=&UniformRangedFloat(Minimum=1, Maximum=2)
    Module.Write Color=true
    Module.Write Lifetime=true
    Module.Write Mass=false
    Module.Write Material Random=true
    Module.Write Position=false
    Module.Write Scale=false
    Module.Write SpriteRotation=true
    Module.Write SpriteSize=true
    Position=&SimulationPosition()
    Sprite Rotation=&UniformRangedFloat(Minimum=0, Maximum=359)
    Sprite Size=&Vector2DFromFloat(Value=&UniformRangedFloat(Minimum=13, Maximum=16))
   )
   &AddVelocity(Coordinate Space=::World, Velocity=&RandomRangeVector(Minimum=<0.1 0.1 -2>, Maximum=<-0.1 -0.1 -5>))
   /SetVariables(Particles.PositionOffset:Vector3f=&Add_Vector(A=&RandomVector(Vector Scale=5), B=<0 0 3>)) {isSetParameters=true}
  ]
  particleUpdateScript [
   &ParticleState()
   &SkeletalMeshLocation(
    Bone / Socket Index=&ModuloInt(A=HlslExpression() {hlslExpression=Particles.UniqueID}, B=2)
    Inherit Velocity=false
    Interpolate Mesh Sampling=false
    MaskSpawnGroup=false
    Mesh Sampling Type=::NewEnumerator5
    Sampled Position Offset=&Lerp_Vector(
     A=$Particles.Initial.PositionOffset
     B=&Add_Vector(A=$Particles.Initial.PositionOffset, B=<0 0 -10>)
     Alpha=$Particles.NormalizedAge
    )
    Sampled Position Offset Coordinate Space:ENiagaraMeshTransforms=::NewEnumerator3
    Skeletal Mesh=$User.SkeletalMesh
    Socket Sampling Mode=::NewEnumerator5
   )
   &SolveForcesAndVelocity(Clamp Velocity=false, Limit Acceleration=false)
   &ScaleSpriteSize(Scale Factor=&Vector2DFromFloat(Value=&FloatFromCurve(FloatCurve=DataInterface() {propertyValues {Curve {keys [
          {interpMode=RCIM_Cubic, leaveTangent=1.1118, tangentMode=RCTM_Break, time=0, value=0.5}
          {arriveTangent=0.321, interpMode=RCIM_Cubic, leaveTangent=0.321, tangentMode=RCTM_User, time=1, value=1.25}
         ]}}})))
   &ScaleColor(
    Scale Alpha=&FloatFromCurve(FloatCurve=DataInterface() {propertyValues {Curve {keys [
         {arriveTangent=3.2151, interpMode=RCIM_Cubic, leaveTangent=3.2151, tangentMode=RCTM_User, time=0, value=0}
         <at 0.5 1 RCIM_Cubic>
         <at 1 0 RCIM_Cubic>
        ]}}})
    ScaleA=true
    ScaleRGB=true
   )
   &SubUVAnimation(
    End Frame=26
    SubUV Animation Mode:ENiagara_SubUVLookupMode=::NewEnumerator1
    SubUV Curve=DataInterface() {propertyValues {Curve {keys [
        {arriveTangent=2.3868, interpMode=RCIM_Cubic, leaveTangent=2.3868, tangentMode=RCTM_User, time=0, value=0}
        {interpMode=RCIM_Cubic, tangentMode=RCTM_User, time=1, value=1}
       ]}}}
   )
   &CameraOffset(Camera Offset Amount=&FloatFromCurve(FloatCurve=DataInterface() {propertyValues {Curve {keys [<at 0 5>,<at 1 15>]}}})) {module=CameraOffset001}
  ]
  rendererClasses [&NiagaraSpriteRendererProperties]
  renderers [{rendererClass=&NiagaraSpriteRendererProperties}]
 }
 Middle {
  emitterUpdateScript [
   &EmitterState()
   &SpawnBurst_Instantaneous(
    Spawn Count=2
    Spawn Probability=&MakeCustomFloatFromBool(
     Bool=&SetBoolByIntComparison(A=$Emitter.LoopCount, B=0, Comparison Type=::NewEnumerator4)
     True Float=1
     False Float=0
    )
    Use Spawn Probability=true
   )
  ]
  particleSpawnScript [
   /SetVariables(Particles.ShardSize:NiagaraFloat=&Multiply_Float(A=1, B=&RandomRangeFloat(Minimum=0.06, Maximum=0.11))) {isSetParameters=true, module=SetVariables_#1}
   /Niagara/Modules/Spawn/Initialization/V2/InitializeParticle(
    Mesh Scale=&MakeVector(
     X=$Particles.ShardSize
     Y=$Particles.ShardSize
     Z=&Add_Float(A=-0.02, B=&Multiply_Float(A=1.5, B=$Particles.ShardSize))
    )
    Mesh Scale Mode=::NewEnumerator4
    Sprite Size Mode=::NewEnumerator3
    Uniform Sprite Size=100
    UsePositionOffset=false
    Write Material Random=true
   )
   &InitialMeshOrientation(
    Orientation Axis=<0 0 1>
    Orientation Vector=<0 0 1>
    Rotation=&Lerp_Vector(
     A=&RandomRangeVector(Minimum=<-0.05 -0.025 -0.05>, Maximum=<0.05 0.025 0.05>)
     B=&RandomRangeVector(Minimum=<-0.05 -0.025 -0.05>, Maximum=<0.05 0.025 0.05>)
     Alpha=&MakeFloatFromInt(Int=&ReturnExecIndex())
    )
    Use Orientation Vector=true
    Use Rotation Vector=true
   )
  ]
  particleUpdateScript [
   &ParticleState(
    Kill Particles When Lifetime Has Elapsed=false
    Let Infinitely Lived Particles Die When Emitter Deactivates=true
    Loop Particles Lifetime=true
   )
   &SkeletalMeshLocation(
    Bone / Socket Index=&ReturnExecIndex()
    Inherit Velocity=false
    Interpolate Mesh Sampling=false
    MaskSpawnGroup=false
    Mesh Sampling Type=::NewEnumerator5
    Sampled Position Offset=&Lerp_Vector(
     A=<3 -2 6.5>
     B=<5 2 5>
     Alpha=&MakeFloatFromInt(Int=&ReturnExecIndex())
    )
    Sampled Position Offset Coordinate Space:ENiagaraMeshTransforms=::NewEnumerator3
    Skeletal Mesh=$User.SkeletalMesh
    Socket Sampling Mode=::NewEnumerator5
   )
   &UpdateMeshOrientation(
    Rotation Coordinate Space=::NewEnumerator3
    Rotation Rate=&UniformAOrBFloat(
     A=&RandomRangeFloat(Minimum=-0.1, Maximum=-0.5)
     B=&RandomRangeFloat(Minimum=0.1, Maximum=0.5)
     Distribution Weight=&ClampFloat(Float=&Round_Float(Float:NiagaraFloat=$Particles.Initial.MaterialRandom))
    )
    Rotation Vector=<0 0 1>
   )
  ]
  rendererClasses [&NiagaraMeshRendererProperties]
  renderers [{rendererClass=&NiagaraMeshRendererProperties, rendererIndex=0}]
 }
}
summary {
 systemName=NS_Skin_FoxWarrior_Arctic_Idle
 emitters [
  {emitterName=BackShard, rendererClasses [&NiagaraMeshRendererProperties]}
  {emitterName=Middle, rendererClasses [&NiagaraMeshRendererProperties]}
  {emitterName=FrontShard, rendererClasses [&NiagaraMeshRendererProperties]}
  {emitterName=ExtraShard, rendererClasses [&NiagaraMeshRendererProperties]}
  {emitterName=FrostyMist_Local, rendererClasses [&NiagaraSpriteRendererProperties]}
 ]
 userVariables [
  User.PhysMatIndex {type=NiagaraInt32}
  User.SkeletalMesh {
   type=NiagaraDataInterfaceSkeletalMesh
   defaultValue {
    struct=NiagaraExt_VariableValue_DataInterface
    value {dataInterface=&self:NiagaraDataInterfaceSkeletalMesh_0}
   }
  }
  User.VariantNum {type=NiagaraInt32, defaultValue=1}
 ]
}
