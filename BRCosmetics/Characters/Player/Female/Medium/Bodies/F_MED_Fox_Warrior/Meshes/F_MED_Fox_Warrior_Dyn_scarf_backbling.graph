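# Physics asset for the Fox Warrior female character's scarf backbling, defining two dynamic scarf tails and body shapes around them.
# Each tail chains from a master bone through limited-swing, low-twist joints, using capsules at the ends and boxes between.
# The left master attaches to spine_05 and the right to clavicle_r; the right chain skips bone 05.
# Torso, clavicle, upper-arm and thigh capsules approximate the body, linked by free angular constraints, as are both masters.
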PhysicsAsset /BRCosmetics/Characters/Player/Female/Medium/Bodies/F_MED_Fox_Warrior/Meshes/F_MED_Fox_Warrior_Dyn_scarf_backbling
bodies {
 clavicle_l [{shapeType=Capsule, length=5.25, radius=6.25, center=<4.103 0.5987 -1.773>, rotation=<rot 90 0 180>}]
 clavicle_r [{shapeType=Capsule, length=5.25, radius=6.25, center=<-4.103 -0.5987 1.773>, rotation=<rot 90 0 0>}]
 dyn_scarf_01_l [{shapeType=Capsule, length=3.5, radius=2, center=<2 0 0>, rotation=<rot 0 0 90>}]
 dyn_scarf_01_r [{shapeType=Capsule, length=3, radius=1.75, center=<2 0 0>, rotation=<rot 0 0 90>}]
 dyn_scarf_02_l [{shapeType=Box, extentX=8, extentY=6.5, extentZ=2.5, center=<5 0 0>}]
 dyn_scarf_02_r [{shapeType=Box, extentX=10, extentY=7, extentZ=2.5, center=<5 0 0>}]
 dyn_scarf_03_l [{shapeType=Box, extentX=10.5, extentY=6.5, extentZ=3, center=<5 0 0>}]
 dyn_scarf_03_r [{shapeType=Box, extentX=10, extentY=6.5, extentZ=3, center=<5 0 -0.3578>}]
 dyn_scarf_04_l [{shapeType=Box, extentX=10.5, extentY=6.5, extentZ=3, center=<5 0 0>}]
 dyn_scarf_04_r [{shapeType=Box, extentX=10, extentY=6.5, extentZ=3, center=<5 0 -0.3578>}]
 dyn_scarf_05_l [{shapeType=Box, extentX=10.5, extentY=6.5, extentZ=3, center=<5 0 0>}]
 dyn_scarf_06_l [{shapeType=Capsule, length=6.75, radius=3.5, center=<9 0 0.5624>, rotation=<rot 90 0 0>}]
 dyn_scarf_06_r [{shapeType=Box, extentX=10, extentY=6.5, extentZ=3, center=<5 0 -0.3578>}]
 dyn_scarf_07_r [{shapeType=Capsule, length=6.75, radius=3.5, center=<9 -2 0.8605>, rotation=<rot 90 -172.875 -162.875>}]
 dyn_scarf_l_MSTR [{shapeType=Capsule, length=1, radius=1}]
 dyn_scarf_r_MSTR [{shapeType=Capsule, length=1, radius=1}]
 pelvis [{shapeType=Capsule, length=10, radius=12.25, center=<-4.8735 -0.8693 0.0001>}]
 spine_01 [{shapeType=Capsule, length=12, radius=8.75, center=<0 0.0001 1.414>}]
 spine_02 [{shapeType=Capsule, length=10.75, radius=8.75, center=<0 1.8581 2.7901>}]
 spine_03 [{shapeType=Capsule, length=9.25, radius=8.75, center=<0 1.8386 0.0001>}]
 spine_04 [{shapeType=Capsule, length=9.25, radius=8.75, center=<0 0.6407 0>}]
 spine_05 [{shapeType=Capsule, length=9.25, radius=8.75, center=<0.5032 0.6468 0>}]
 thigh_l [{shapeType=Capsule, length=32.2805, radius=8.5564, center=<-20.2752 0 0.0003>, rotation=<rot 90 1.7899 -175.1438>}]
 thigh_r [{shapeType=Capsule, length=32.2805, radius=8.5564, center=<20.2752 0 -0.0003>, rotation=<rot 90 -1.7899 1.2764>}]
 upperarm_l [{shapeType=Capsule, length=24.25, radius=4.75, center=<12.8326 -0.1046 -0.0001>, rotation=<rot 90 5.0512 -174.9487>}]
 upperarm_r [{shapeType=Capsule, length=24.25, radius=4.75, center=<-12.8326 0.1046 0.0001>, rotation=<rot 90 -6.7213 -6.7213>}]
}
constraints [
 {
  bone1Name=dyn_scarf_r_MSTR
  bone2Name=clavicle_r
  swing1LimitDegrees=45
  swing1Motion=Free
  swing2LimitDegrees=45
  swing2Motion=Free
  twistLimitDegrees=45
  twistMotion=Free
 }
 {bone1Name=dyn_scarf_01_r, bone2Name=dyn_scarf_r_MSTR, swing1LimitDegrees=10, swing2LimitDegrees=5, twistLimitDegrees=5}
 {bone1Name=dyn_scarf_02_r, bone2Name=dyn_scarf_01_r, swing1LimitDegrees=20, swing2LimitDegrees=20, twistLimitDegrees=2}
 {bone1Name=dyn_scarf_03_r, bone2Name=dyn_scarf_02_r, swing1LimitDegrees=20, swing2LimitDegrees=20, twistLimitDegrees=2}
 {bone1Name=dyn_scarf_04_r, bone2Name=dyn_scarf_03_r, swing1LimitDegrees=20, swing2LimitDegrees=20, twistLimitDegrees=2}
 {
  bone1Name=clavicle_r
  bone2Name=spine_05
  swing1LimitDegrees=45
  swing1Motion=Free
  swing2LimitDegrees=45
  swing2Motion=Free
  twistLimitDegrees=45
  twistMotion=Free
 }
 {
  bone1Name=spine_05
  bone2Name=spine_04
  swing1LimitDegrees=45
  swing1Motion=Free
  swing2LimitDegrees=45
  swing2Motion=Free
  twistLimitDegrees=45
  twistMotion=Free
 }
 {
  bone1Name=spine_04
  bone2Name=spine_03
  swing1LimitDegrees=45
  swing1Motion=Free
  swing2LimitDegrees=45
  swing2Motion=Free
  twistLimitDegrees=45
  twistMotion=Free
 }
 {
  bone1Name=spine_03
  bone2Name=spine_02
  swing1LimitDegrees=45
  swing1Motion=Free
  swing2LimitDegrees=45
  swing2Motion=Free
  twistLimitDegrees=45
  twistMotion=Free
 }
 {
  bone1Name=spine_02
  bone2Name=spine_01
  swing1LimitDegrees=45
  swing1Motion=Free
  swing2LimitDegrees=45
  swing2Motion=Free
  twistLimitDegrees=45
  twistMotion=Free
 }
 {
  bone1Name=spine_01
  bone2Name=pelvis
  swing1LimitDegrees=45
  swing1Motion=Free
  swing2LimitDegrees=45
  swing2Motion=Free
  twistLimitDegrees=45
  twistMotion=Free
 }
 {
  bone1Name=dyn_scarf_l_MSTR
  bone2Name=spine_05
  swing1LimitDegrees=45
  swing1Motion=Free
  swing2LimitDegrees=45
  swing2Motion=Free
  twistLimitDegrees=45
  twistMotion=Free
 }
 {bone1Name=dyn_scarf_01_l, bone2Name=dyn_scarf_l_MSTR, swing1LimitDegrees=10, swing2LimitDegrees=5, twistLimitDegrees=5}
 {bone1Name=dyn_scarf_02_l, bone2Name=dyn_scarf_01_l, swing1LimitDegrees=20, swing2LimitDegrees=20, twistLimitDegrees=2}
 {bone1Name=dyn_scarf_03_l, bone2Name=dyn_scarf_02_l, swing1LimitDegrees=20, swing2LimitDegrees=20, twistLimitDegrees=2}
 {bone1Name=dyn_scarf_04_l, bone2Name=dyn_scarf_03_l, swing1LimitDegrees=20, swing2LimitDegrees=20, twistLimitDegrees=2}
 {
  bone1Name=thigh_r
  bone2Name=pelvis
  swing1LimitDegrees=45
  swing1Motion=Free
  swing2LimitDegrees=45
  swing2Motion=Free
  twistLimitDegrees=45
  twistMotion=Free
 }
 {
  bone1Name=thigh_l
  bone2Name=pelvis
  swing1LimitDegrees=45
  swing1Motion=Free
  swing2LimitDegrees=45
  swing2Motion=Free
  twistLimitDegrees=45
  twistMotion=Free
 }
 {
  bone1Name=upperarm_r
  bone2Name=clavicle_r
  swing1LimitDegrees=45
  swing1Motion=Free
  swing2LimitDegrees=45
  swing2Motion=Free
  twistLimitDegrees=45
  twistMotion=Free
 }
 {bone1Name=dyn_scarf_06_r, bone2Name=dyn_scarf_04_r, swing1LimitDegrees=20, swing2LimitDegrees=20, twistLimitDegrees=2}
 {bone1Name=dyn_scarf_07_r, bone2Name=dyn_scarf_06_r, swing1LimitDegrees=20, swing2LimitDegrees=20, twistLimitDegrees=5}
 {bone1Name=dyn_scarf_05_l, bone2Name=dyn_scarf_04_l, swing1LimitDegrees=20, swing2LimitDegrees=20, twistLimitDegrees=5}
 {bone1Name=dyn_scarf_06_l, bone2Name=dyn_scarf_05_l, swing1LimitDegrees=20, swing2LimitDegrees=20, twistLimitDegrees=5}
 {
  bone1Name=upperarm_l
  bone2Name=clavicle_l
  swing1LimitDegrees=45
  swing1Motion=Free
  swing2LimitDegrees=45
  swing2Motion=Free
  twistLimitDegrees=45
  twistMotion=Free
 }
 {
  bone1Name=clavicle_l
  bone2Name=spine_05
  swing1LimitDegrees=45
  swing1Motion=Free
  swing2LimitDegrees=45
  swing2Motion=Free
  twistLimitDegrees=45
  twistMotion=Free
 }
]
