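# Physics asset for secondary motion of the skirt, chain, bags and bottles on the female medium Fox Warrior cosmetic body.
# Small MSTR anchor bodies join the pelvis with free angular constraints, except the back-left skirt anchor, which has no parent constraint.
# Short chains of box and capsule bodies hang from the anchors with tight swing and twist limits.
# Pelvis and thigh capsules supply the body volumes beneath them.
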PhysicsAsset /BRCosmetics/Characters/Player/Female/Medium/Bodies/F_MED_Fox_Warrior/Meshes/F_MED_Fox_Warrior_dyn_skirt
bodies {
 dyn_bag_01 [{shapeType=Capsule, length=4, radius=3, center=<5 0 1.5>, rotation=<rot 90 0 0>}]
 dyn_bag_01_MSTR [{shapeType=Capsule, length=0.505, radius=0.505, rotation=<rot 0 -90 0>}]
 dyn_bag_02 [{shapeType=Capsule, length=4, radius=3, center=<5 0 1.5>, rotation=<rot 90 0 0>}]
 dyn_bag_02_MSTR [{shapeType=Capsule, length=0.505, radius=0.505, rotation=<rot 0 -90 0>}]
 dyn_bottle_01 [{shapeType=Capsule, length=5.5, radius=2.5, center=<5.2 0 1.5>, rotation=<rot 85 0 0>}]
 dyn_bottle_01_MSTR [{shapeType=Capsule, length=0.505, radius=0.505, rotation=<rot 0 -90 0>}]
 dyn_bottle_02 [{shapeType=Capsule, length=5.5, radius=2.5, center=<5.2 0 1.5>, rotation=<rot 85 0 0>}]
 dyn_bottle_02_MSTR [{shapeType=Capsule, length=0.505, radius=0.505, rotation=<rot 0 -90 0>}]
 dyn_bottle_03 [{shapeType=Capsule, length=5.5, radius=2.5, center=<5.2 0 1.5>, rotation=<rot 85 0 0>}]
 dyn_bottle_03_MSTR [{shapeType=Capsule, length=0.505, radius=0.505, rotation=<rot 0 -90 0>}]
 dyn_chain_01 [{shapeType=Box, extentX=7.5, extentY=17, extentZ=2, center=<4.6704 -0.0327 0.3722>, rotation=<rot -3.8429 0.3367 -5.0115>}]
 dyn_chain_02 [{shapeType=Capsule, length=2.75, radius=1.5, center=<3.6698 0 0.3265>, rotation=<rot 90 0 0>}]
 dyn_chain_MSTR [{shapeType=Capsule, length=0.505, radius=0.505, rotation=<rot 0 -90 0>}]
 dyn_skirt_01_ft [{shapeType=Box, extentX=10.5, extentY=15.5, extentZ=2.5, center=<5 0 0>}]
 dyn_skirt_01_r [{shapeType=Capsule, length=6.75, radius=4.75, shapeName=dyn_skirt_01_r_capsule, center=<6 0 4>, rotation=<rot 0 0 90>}]
 dyn_skirt_02_ft [{shapeType=Box, extentX=10.5, extentY=15.5, extentZ=2.5, center=<5 0 0>}]
 dyn_skirt_02_r [{shapeType=Capsule, length=6.5, radius=5, shapeName=dyn_skirt_02_r_capsule, center=<4 0 4>, rotation=<rot -0.0001 10 90>}]
 dyn_skirt_03_r [{shapeType=Capsule, length=12.75, radius=4.75, shapeName=dyn_skirt_03_r_capsule, center=<4 0 5>, rotation=<rot -0.0001 20 90>}]
 dyn_skirt_bk_01_l [{shapeType=Capsule, length=6.5, radius=5, shapeName=dyn_skirt_02_r_capsule, center=<10 0 3>, rotation=<rot -0.0001 0 100>}]
 dyn_skirt_bk_l_MSTR [{shapeType=Sphere, radius=0.5001}]
 dyn_skirt_ft_MSTR [{shapeType=Capsule, length=0.505, radius=0.505, rotation=<rot 0 -90 0>}]
 dyn_skirt_r_MSTR [{shapeType=Capsule, length=0.505, radius=0.505, rotation=<rot 0 -90 0>}]
 pelvis [{shapeType=Capsule, length=8.5, radius=10.9849, center=<-5.4818 -2.0706 0>, rotation=<rot 0 -90 0>}]
 thigh_l [{shapeType=Capsule, length=30, radius=8, center=<-18.9994 -0.0782 0.0004>, rotation=<rot 90 5.3558 -169.6441>}]
 thigh_r [{shapeType=Capsule, length=26.9862, radius=9.2638, center=<19.0109 0.6571 -0.0004>, rotation=<rot 90 -0.2238 0.9149>}]
}
constraints [
 {
  bone1Name=thigh_r
  bone2Name=pelvis
  swing1LimitDegrees=45
  swing1Motion=Free
  swing2LimitDegrees=45
  swing2Motion=Free
  twistLimitDegrees=45
  twistMotion=Free
 }
 {
  bone1Name=dyn_chain_MSTR
  bone2Name=pelvis
  swing1LimitDegrees=45
  swing1Motion=Free
  swing2LimitDegrees=45
  swing2Motion=Free
  twistLimitDegrees=45
  twistMotion=Free
 }
 {bone1Name=dyn_chain_01, bone2Name=dyn_chain_MSTR, swing1LimitDegrees=10, swing2LimitDegrees=30, twistLimitDegrees=2}
 {bone1Name=dyn_chain_02, bone2Name=dyn_chain_01, swing1LimitDegrees=10, swing2LimitDegrees=30, twistLimitDegrees=2}
 {
  bone1Name=thigh_l
  bone2Name=pelvis
  swing1LimitDegrees=45
  swing1Motion=Free
  swing2LimitDegrees=45
  swing2Motion=Free
  twistLimitDegrees=45
  twistMotion=Free
 }
 {
  bone1Name=dyn_skirt_r_MSTR
  bone2Name=pelvis
  swing1LimitDegrees=45
  swing1Motion=Free
  swing2LimitDegrees=45
  swing2Motion=Free
  twistLimitDegrees=45
  twistMotion=Free
 }
 {bone1Name=dyn_skirt_01_r, bone2Name=dyn_skirt_r_MSTR, swing1LimitDegrees=10, swing2LimitDegrees=20, twistLimitDegrees=2}
 {
  bone1Name=dyn_bag_01_MSTR
  bone2Name=pelvis
  swing1LimitDegrees=45
  swing1Motion=Free
  swing2LimitDegrees=45
  swing2Motion=Free
  twistLimitDegrees=45
  twistMotion=Free
 }
 {bone1Name=dyn_bag_01, bone2Name=dyn_bag_01_MSTR, swing1LimitDegrees=15, swing2LimitDegrees=10, twistLimitDegrees=2}
 {
  bone1Name=dyn_bag_02_MSTR
  bone2Name=pelvis
  swing1LimitDegrees=45
  swing1Motion=Free
  swing2LimitDegrees=45
  swing2Motion=Free
  twistLimitDegrees=45
  twistMotion=Free
 }
 {bone1Name=dyn_bag_02, bone2Name=dyn_bag_02_MSTR, swing1LimitDegrees=15, swing2LimitDegrees=10, twistLimitDegrees=2}
 {
  bone1Name=dyn_bottle_01_MSTR
  bone2Name=pelvis
  swing1LimitDegrees=45
  swing1Motion=Free
  swing2LimitDegrees=45
  swing2Motion=Free
  twistLimitDegrees=45
  twistMotion=Free
 }
 {bone1Name=dyn_bottle_01, bone2Name=dyn_bottle_01_MSTR, swing1LimitDegrees=10, swing2LimitDegrees=15, twistLimitDegrees=5}
 {
  bone1Name=dyn_bottle_02_MSTR
  bone2Name=pelvis
  swing1LimitDegrees=45
  swing1Motion=Free
  swing2LimitDegrees=45
  swing2Motion=Free
  twistLimitDegrees=45
  twistMotion=Free
 }
 {bone1Name=dyn_bottle_02, bone2Name=dyn_bottle_02_MSTR, swing1LimitDegrees=10, swing2LimitDegrees=15, twistLimitDegrees=5}
 {
  bone1Name=dyn_bottle_03_MSTR
  bone2Name=pelvis
  swing1LimitDegrees=45
  swing1Motion=Free
  swing2LimitDegrees=45
  swing2Motion=Free
  twistLimitDegrees=45
  twistMotion=Free
 }
 {bone1Name=dyn_bottle_03, bone2Name=dyn_bottle_03_MSTR, swing1LimitDegrees=10, swing2LimitDegrees=15, twistLimitDegrees=5}
 {
  bone1Name=dyn_skirt_ft_MSTR
  bone2Name=pelvis
  swing1LimitDegrees=45
  swing1Motion=Free
  swing2LimitDegrees=45
  swing2Motion=Free
  twistLimitDegrees=45
  twistMotion=Free
 }
 {bone1Name=dyn_skirt_01_ft, bone2Name=dyn_skirt_ft_MSTR, swing1LimitDegrees=20, swing2LimitDegrees=20, twistLimitDegrees=2}
 {bone1Name=dyn_skirt_02_ft, bone2Name=dyn_skirt_01_ft, swing1LimitDegrees=20, swing2LimitDegrees=20, twistLimitDegrees=2}
 {bone1Name=dyn_skirt_02_r, bone2Name=dyn_skirt_01_r, swing1LimitDegrees=10, swing2LimitDegrees=20, twistLimitDegrees=2}
 {bone1Name=dyn_skirt_bk_01_l, bone2Name=dyn_skirt_bk_l_MSTR, swing1LimitDegrees=5, swing2LimitDegrees=20, twistLimitDegrees=2}
 {bone1Name=dyn_skirt_03_r, bone2Name=dyn_skirt_02_r, swing1LimitDegrees=10, swing2LimitDegrees=20, twistLimitDegrees=2}
]
