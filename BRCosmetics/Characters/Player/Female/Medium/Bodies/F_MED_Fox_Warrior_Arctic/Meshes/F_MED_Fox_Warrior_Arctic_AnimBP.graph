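# Post-process style cosmetic AnimBP for the Arctic Fox Warrior body, adding secondary motion to scarf, skirt, chain, bags and bottles.
# It copies the attached parent's pose, constrains accessories and caches it.
# RigidBody physics varies by backbling, with posed scarves for capes and Look At/Trail nodes when physics is disabled or LOD exceeds 1.
# Event graph derives thigh-driven corrective alphas, crouch and gravity for swim-sprinting or skydiving.
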AnimBlueprint /BRCosmetics/Characters/Player/Female/Medium/Bodies/F_MED_Fox_Warrior_Arctic/Meshes/F_MED_Fox_Warrior_Arctic_AnimBP
target_skeleton=&/F_MED_Fox_Warrior_Arctic_Skeleton
graphs [
 AnimGraph {nodes [
   n1 = Get thigh_rot_up_r() -> (thigh_rot_up_r:double)
   n2 = "Transform (Modify) Bone - Bone: dyn_skirt_01_ft"(Alpha=@n1, Translation="0,0,0", Rotation="10,0,0", Scale="0.85,1.000,1.000") {
    rotationMode=BMM_Additive
    rotationSpace=BCS_BoneSpace
    scaleMode=BMM_Additive
    scaleSpace=BCS_BoneSpace
    translationMode=BMM_Additive
    translationSpace=BCS_BoneSpace
    alphaBoolBlend {blendOption=Linear}
    boneToModify {boneName=dyn_skirt_01_ft}
    rotation {pitch=10}
    scale {x=0.85}
   }
   "Transform (Modify) Bone - Bone: dyn_skirt_01_r"(ComponentPose=@n2, Alpha=@n1, Translation="0,0,0", Rotation="50,0,0", Scale="1,1,1") {
    rotationMode=BMM_Additive
    rotationSpace=BCS_BoneSpace
    scaleMode=BMM_Additive
    scaleSpace=BCS_BoneSpace
    translationMode=BMM_Additive
    translationSpace=BCS_BoneSpace
    alphaBoolBlend {blendOption=Linear}
    boneToModify {boneName=dyn_skirt_01_r}
    rotation {pitch=50}
   }
   n3 = "Transform (Modify) Bone - Bone: dyn_scarf_01_l"(
    ComponentPose=Local To Component(LocalPose=Use cached pose 'prePose'())
    Alpha=1
    Translation="0,0,0"
    Rotation="-25,0.000,0.000"
    Scale="1,1,1"
   ) {
    rotationMode=BMM_Additive
    rotationSpace=BCS_BoneSpace
    alphaBoolBlend {blendOption=Linear}
    boneToModify {boneName=dyn_scarf_01_l}
    rotation {pitch=-25}
   }
   n4 = "Transform (Modify) Bone - Bone: dyn_scarf_01_r"(ComponentPose=@n3, Alpha=1, Translation="0,0,0", Rotation="-25,0,0", Scale="1,1,1") {
    rotationMode=BMM_Additive
    rotationSpace=BCS_BoneSpace
    alphaBoolBlend {blendOption=Linear}
    boneToModify {boneName=dyn_scarf_01_r}
    rotation {pitch=-25}
   }
   n5 = "Constraint - dyn_scarf_03_l (RR)"!(ComponentPose=@n4, Alpha=1, ConstraintWeights_0=0.2, ConstraintWeights_1:float=0.8) -> (Pose) {
    alphaBoolBlend {blendOption=Linear}
    boneToModify {boneName=dyn_scarf_03_l}
    constraintSetup [{transformType=Rotation, targetBone {boneName=dyn_scarf_03_l}}, {transformType=Rotation, targetBone {boneName=pelvis}}]
    constraintWeights [1,1]
   }
   n6 = "Transform (Modify) Bone - Bone: dyn_scarf_03_l"(ComponentPose=@n5, Alpha=1, Translation="0,0,0", Rotation="-20,0,0.000", Scale="1,1,1") {
    rotationMode=BMM_Additive
    rotationSpace=BCS_BoneSpace
    alphaBoolBlend {blendOption=Linear}
    boneToModify {boneName=dyn_scarf_03_l}
    rotation {pitch=-20}
   }
   n7 = "Constraint - dyn_scarf_03_r (RR)"!(ComponentPose=@n6, Alpha=1, ConstraintWeights_0=0.2, ConstraintWeights_1:float=0.8) -> (Pose) {
    alphaBoolBlend {blendOption=Linear}
    boneToModify {boneName=dyn_scarf_03_r}
    constraintSetup [{transformType=Rotation, targetBone {boneName=dyn_scarf_03_r}}, {transformType=Rotation, targetBone {boneName=pelvis}}]
    constraintWeights [1,1]
   }
   n8 = "Transform (Modify) Bone - Bone: dyn_scarf_03_r"(ComponentPose=@n7, Alpha=1, Translation="0,0,1", Rotation="-20,-10,0", Scale="1,1,1") {
    rotationMode=BMM_Additive
    rotationSpace=BCS_BoneSpace
    translationMode=BMM_Additive
    translationSpace=BCS_BoneSpace
    alphaBoolBlend {blendOption=Linear}
    boneToModify {boneName=dyn_scarf_03_r}
    rotation {pitch=-20, yaw=-10}
    translation {z=1}
   }
   n9 = "Transform (Modify) Bone - Bone: dyn_scarf_03_r"(ComponentPose=@n8, Alpha=Get crouch() -> (crouch:double), Translation="0,0,0", Rotation="5,0,0", Scale="1,1,1") {
    rotationMode=BMM_Additive
    rotationSpace=BCS_BoneSpace
    alphaBoolBlend {blendOption=Linear}
    boneToModify {boneName=dyn_scarf_03_r}
    rotation {pitch=5}
   }
   n10 = "Constraint - dyn_scarf_05_l (PP)"!(ComponentPose=@n9, Alpha=1, ConstraintWeights_0=0.8, ConstraintWeights_1:float=0.2) -> (Pose) {
    alphaBoolBlend {blendOption=Linear}
    boneToModify {boneName=dyn_scarf_05_l}
    constraintSetup [{targetBone {boneName=dyn_scarf_05_l}}, {targetBone {boneName=thigh_twist_01_l}}]
    constraintWeights [1,1]
   }
   n11 = "Constraint - dyn_scarf_05_l (RR)"!(ComponentPose=@n10, Alpha=1, ConstraintWeights_0=0.8, ConstraintWeights_1:float=0.2) -> (Pose) {
    alphaBoolBlend {blendOption=Linear}
    boneToModify {boneName=dyn_scarf_05_l}
    constraintSetup [{transformType=Rotation, targetBone {boneName=dyn_scarf_05_l}}, {transformType=Rotation, targetBone {boneName=thigh_twist_01_l}}]
    constraintWeights [1,1]
   }
   n12 = "Transform (Modify) Bone - Bone: dyn_scarf_05_l"(ComponentPose=@n11, Alpha=1, Translation="0,0,0", Rotation="-25,0.000,0.000", Scale="1,1,1") {
    rotationMode=BMM_Additive
    rotationSpace=BCS_BoneSpace
    alphaBoolBlend {blendOption=Linear}
    boneToModify {boneName=dyn_scarf_05_l}
    rotation {pitch=-25}
   }
   n13 = "Constraint - dyn_scarf_06_r (PP)"!(ComponentPose=@n12, Alpha=1, ConstraintWeights_0=0.8, ConstraintWeights_1:float=0.2) -> (Pose) {
    alphaBoolBlend {blendOption=Linear}
    boneToModify {boneName=dyn_scarf_06_r}
    constraintSetup [{targetBone {boneName=dyn_scarf_06_r}}, {targetBone {boneName=thigh_twist_01_r}}]
    constraintWeights [1,1]
   }
   n14 = "Constraint - dyn_scarf_06_r (RR)"!(ComponentPose=@n13, Alpha=1, ConstraintWeights_0=0.8, ConstraintWeights_1:float=0.2) -> (Pose) {
    alphaBoolBlend {blendOption=Linear}
    boneToModify {boneName=dyn_scarf_06_r}
    constraintSetup [{transformType=Rotation, targetBone {boneName=dyn_scarf_06_r}}, {transformType=Rotation, targetBone {boneName=thigh_twist_01_r}}]
    constraintWeights [1,1]
   }
   n15 = "Transform (Modify) Bone - Bone: dyn_scarf_06_r"(ComponentPose=@n14, Alpha=1, Translation="0,0,0", Rotation="-20,0,0", Scale="1,1,1") {
    rotationMode=BMM_Additive
    rotationSpace=BCS_BoneSpace
    alphaBoolBlend {blendOption=Linear}
    boneToModify {boneName=dyn_scarf_06_r}
    rotation {pitch=-20}
   }
   n16 = Get gravity_override() -> (gravity_override:Vector)
   n17 = RigidBody!(ComponentPose=Local To Component(LocalPose=Use cached pose 'prePose'()), Alpha, OverrideWorldGravity:Vector=@n16, ExternalForce) -> (Pose) {
    bClampLinearTranslationLimitToRefPose=true
    bOverrideWorldGravity=true
    overridePhysicsAsset=/BRCosmetics/Characters/Player/Female/Medium/Bodies/F_MED_Fox_Warrior/Meshes/F_MED_Fox_Warrior_Dyn_scarf
    simulationSpace=BaseBoneSpace
    alphaBoolBlend {blendOption=Linear}
    baseBoneRef {boneName=root}
    componentAppliedLinearAccClamp=<1000 1000 3000>
    componentLinearAccScale=<0.2 0.2 0.45>
    componentLinearVelScale=<0.1 0.1 0.1>
    overrideWorldGravity {z=-980}
   }
   n18 = Component To Local(ComponentPose=@n17)
   n19 = "Transform (Modify) Bone - Bone: dyn_scarf_01_l"(
    ComponentPose=Local To Component(LocalPose=Use cached pose 'prePose'())
    Alpha=1
    Translation="0,0,0"
    Rotation="-25,0.000,0.000"
    Scale="1,1,1"
   ) {
    rotationMode=BMM_Additive
    rotationSpace=BCS_BoneSpace
    alphaBoolBlend {blendOption=Linear}
    boneToModify {boneName=dyn_scarf_01_l}
    rotation {pitch=-25}
   }
   n20 = "Transform (Modify) Bone - Bone: dyn_scarf_01_r"(ComponentPose=@n19, Alpha=1, Translation="0,0,0", Rotation="-25,0,0", Scale="1,1,1") {
    rotationMode=BMM_Additive
    rotationSpace=BCS_BoneSpace
    alphaBoolBlend {blendOption=Linear}
    boneToModify {boneName=dyn_scarf_01_r}
    rotation {pitch=-25}
   }
   n21 = RigidBody(ComponentPose=@n20, OverrideWorldGravity=@n16) {
    bClampLinearTranslationLimitToRefPose=true
    bOverrideWorldGravity=true
    overridePhysicsAsset=/BRCosmetics/Characters/Player/Female/Medium/Bodies/F_MED_Fox_Warrior/Meshes/F_MED_Fox_Warrior_Dyn_scarf_backbling
    simulationSpace=BaseBoneSpace
    alphaBoolBlend {blendOption=Linear}
    baseBoneRef {boneName=root}
    componentAppliedLinearAccClamp=<3000 3000 3000>
    componentLinearAccScale=<0.1 0.1 0.1>
    componentLinearVelScale=<0.5 0.5 0.5>
    overrideWorldGravity {z=-980}
   }
   n22 = Component To Local(ComponentPose=@n21)
   n23 = Get isWearingBackbling()
   n24 = RigidBody!(
    ComponentPose=Local To Component(LocalPose=Blend Poses by bool(BlendPose_0=@n22, BlendPose_1=@n18, bActiveValue=@n23) {blendType=Linear})
    Alpha
    ExternalForce
   ) -> (Pose) {
    bClampLinearTranslationLimitToRefPose=true
    bOverrideWorldGravity=true
    overridePhysicsAsset=/BRCosmetics/Characters/Player/Female/Medium/Bodies/F_MED_Fox_Warrior/Meshes/F_MED_Fox_Warrior_dyn_skirt
    simulationSpace=BaseBoneSpace
    alphaBoolBlend {blendOption=Linear}
    baseBoneRef {boneName=root}
    componentAppliedLinearAccClamp=<1000 1000 1000>
    componentLinearAccScale=<0.2 0.2 0.2>
    componentLinearVelScale=<0.2 0.2 0.2>
    overrideWorldGravity {z=-980}
   }
   n25 = Blend Poses by bool(
    BlendPose_0=Component To Local(ComponentPose=@n15)
    BlendPose_1=Component To Local(ComponentPose=@n24)
    bActiveValue=Get Is Backbling A Cape() -> (Is Backbling A Cape:bool)
   ) {blendType=Linear}
   n26 = Get thigh_rot_up_r()
   n27 = "Transform (Modify) Bone - Bone: dyn_chain_01"(ComponentPose=Local To Component(LocalPose=Use cached pose 'prePose'()), Alpha=@n26, Translation="0,0,0", Rotation="50,0,20", Scale="1,1,1") {
    rotationMode=BMM_Additive
    rotationSpace=BCS_BoneSpace
    scaleMode=BMM_Additive
    scaleSpace=BCS_BoneSpace
    translationMode=BMM_Additive
    translationSpace=BCS_BoneSpace
    alphaBoolBlend {blendOption=Linear}
    boneToModify {boneName=dyn_chain_01}
    rotation {pitch=50, roll=20}
   }
   n28 = "Transform (Modify) Bone - Bone: dyn_chain_02"(ComponentPose=@n27, Alpha=@n26, Translation="0,0,0", Rotation="45,0,5", Scale="1,1,1") {
    rotationMode=BMM_Additive
    rotationSpace=BCS_BoneSpace
    scaleMode=BMM_Additive
    scaleSpace=BCS_BoneSpace
    translationMode=BMM_Additive
    translationSpace=BCS_BoneSpace
    alphaBoolBlend {blendOption=Linear}
    boneToModify {boneName=dyn_chain_02}
    rotation {pitch=45, roll=5}
   }
   n29 = "Transform (Modify) Bone - Bone: dyn_scarf_01_l"(ComponentPose=@n28, Alpha=1, Translation="0,0,0", Rotation="-20,0.000,0.000", Scale="1,1,1") {
    rotationMode=BMM_Additive
    rotationSpace=BCS_BoneSpace
    alphaBoolBlend {blendOption=Linear}
    boneToModify {boneName=dyn_scarf_01_l}
    rotation {pitch=-20}
   }
   n30 = "Transform (Modify) Bone - Bone: dyn_scarf_01_r"(ComponentPose=@n29, Alpha=1, Translation="0,0,0", Rotation="-20,-15.000,0.000", Scale="1,1,1") {
    rotationMode=BMM_Additive
    rotationSpace=BCS_BoneSpace
    alphaBoolBlend {blendOption=Linear}
    boneToModify {boneName=dyn_scarf_01_r}
    rotation {pitch=-20, yaw=-15}
   }
   n31 = Look At - Bone: dyn_scarf_02_r!(ComponentPose=@n30, Alpha=1, LookAtLocation:Vector="0,-14,13") -> (Pose) {
    alphaBoolBlend {blendOption=Linear}
    boneToModify {boneName=dyn_scarf_02_r}
    lookAtLocation=<0 -10>
    lookAtTarget {boneReference {boneName=pelvis}}
    lookAt_Axis {axis=<1 0>}
   }
   n32 = Look At - Bone: dyn_scarf_02_l!(ComponentPose=@n31, Alpha=1, LookAtLocation:Vector="0,-14,-10") -> (Pose) {
    alphaBoolBlend {blendOption=Linear}
    boneToModify {boneName=dyn_scarf_02_l}
    lookAtLocation=<0 -10>
    lookAtTarget {boneReference {boneName=pelvis}}
    lookAt_Axis {axis=<1 0>}
   }
   n33 = Look At - Bone: dyn_scarf_03_r!(ComponentPose=@n32, Alpha=1, LookAtLocation:Vector="0,-16,15") -> (Pose) {
    alphaBoolBlend {blendOption=Linear}
    boneToModify {boneName=dyn_scarf_03_r}
    lookAtLocation=<0 -10>
    lookAtTarget {boneReference {boneName=pelvis}}
    lookAt_Axis {axis=<1 0>}
   }
   n34 = Look At - Bone: dyn_scarf_03_l!(ComponentPose=@n33, Alpha=1, LookAtLocation:Vector="0,-16,-10") -> (Pose) {
    alphaBoolBlend {blendOption=Linear}
    boneToModify {boneName=dyn_scarf_03_l}
    lookAtLocation=<0 -10>
    lookAtTarget {boneReference {boneName=pelvis}}
    lookAt_Axis {axis=<1 0>}
   }
   n35 = Trail controller - Bone: dyn_scarf_06_l(ComponentPose=@n34, Alpha=0.6) {
    bUsePlanarLimit=true
    chainLength=4
    relaxationSpeedScale=4
    planarLimits [{drivingBone {boneName=pelvis}, planeTransform {location=<0 -10 0>, rotation=<rot 0 0 -90>, scale=<1 1 1>}}]
    rotationLimits [{limitMax=<180 180 180>}, {limitMax=<180 180 180>}, {limitMax=<180 180 180>}, {limitMax=<180 180 180>}]
    rotationOffsets [<0 0 0>, <0 0 0>, <0 0 0>, <0 0 0>]
    trailBone {boneName=dyn_scarf_06_l}
    trailRelaxationSpeed {editorCurveData {keys [
       {arriveTangent=-20.0693, interpMode=RCIM_Cubic, leaveTangent=0.752, tangentMode=RCTM_User, time=0, value=15}
       {arriveTangent=-8.3227, interpMode=RCIM_Cubic, leaveTangent=-10.3098, tangentMode=RCTM_User, time=1, value=6}
      ]}}
   }
   n36 = Trail controller - Bone: dyn_scarf_07_r(ComponentPose=@n35, Alpha=0.6) {
    bUsePlanarLimit=true
    chainLength=4
    relaxationSpeedScale=3
    planarLimits [{drivingBone {boneName=pelvis}, planeTransform {location=<0 -8 0>, rotation=<rot 0 0 -90>, scale=<1 1 1>}}]
    rotationLimits [{limitMax=<180 180 180>}, {limitMax=<180 180 180>}, {limitMax=<180 180 180>}, {limitMax=<180 180 180>}]
    rotationOffsets [<0 0 0>, <0 0 0>, <0 0 0>, <0 0 0>]
    trailBone {boneName=dyn_scarf_07_r}
    trailRelaxationSpeed {editorCurveData {keys [
       {arriveTangent=-20.0693, interpMode=RCIM_Cubic, leaveTangent=0.752, tangentMode=RCTM_User, time=0, value=20}
       {arriveTangent=-8.3227, interpMode=RCIM_Cubic, leaveTangent=-10.3098, tangentMode=RCTM_User, time=1, value=6}
      ]}}
   }
   n37 = Component To Local(ComponentPose=@n36)
   n38 = Get isRigidBodyDisabled() -> (isRigidBodyDisabled:bool)
   n39 = Get thigh_rot_up_l() -> (thigh_rot_up_l:double)
   n40 = "Transform (Modify) Bone - Bone: dyn_skirt_bk_01_l"(
    ComponentPose=Local To Component(LocalPose=Blend Poses by bool(BlendPose_0=@n37, BlendPose_1=@n25, bActiveValue=@n38) {blendType=Linear})
    Alpha=@n39
    Translation="0,0,0"
    Rotation="60,0.000,0.000"
    Scale="0.5,1.000,1.000"
   ) {
    rotationMode=BMM_Additive
    rotationSpace=BCS_BoneSpace
    scaleMode=BMM_Additive
    scaleSpace=BCS_BoneSpace
    translationMode=BMM_Additive
    translationSpace=BCS_BoneSpace
    alphaBoolBlend {blendOption=Linear}
    boneToModify {boneName=dyn_skirt_bk_01_l}
    rotation {pitch=60}
    scale {x=0.5}
   }
   n41 = "Transform (Modify) Bone - Bone: dyn_skirt_01_ft"(ComponentPose=@n40, Alpha=@n39, Translation="0,0,0", Rotation="5,-5,0.000", Scale="0.8,1.000,1.000") {
    rotationMode=BMM_Additive
    rotationSpace=BCS_BoneSpace
    scaleMode=BMM_Additive
    scaleSpace=BCS_BoneSpace
    translationMode=BMM_Additive
    translationSpace=BCS_BoneSpace
    alphaBoolBlend {blendOption=Linear}
    boneToModify {boneName=dyn_skirt_01_ft}
    rotation {pitch=5, yaw=-5}
    scale {x=0.8}
   }
   n42 = Get thigh_rot_in_l() -> (thigh_rot_in_l:double)
   n43 = "Transform (Modify) Bone - Bone: dyn_skirt_bk_01_l"(ComponentPose=@n41, Alpha=@n42, Translation="0,0,0", Rotation="5,20,5", Scale="1,1,1") {
    rotationMode=BMM_Additive
    rotationSpace=BCS_BoneSpace
    scaleMode=BMM_Additive
    scaleSpace=BCS_BoneSpace
    translationMode=BMM_Additive
    translationSpace=BCS_BoneSpace
    alphaBoolBlend {blendOption=Linear}
    boneToModify {boneName=dyn_skirt_bk_01_l}
    rotation=<rot 5 20 5>
   }
   n44 = "Transform (Modify) Bone - Bone: dyn_skirt_01_r"(ComponentPose=@n43, Alpha=@n42, Translation="0,0,0", Rotation="5,-20,-10", Scale="1,1,1") {
    rotationMode=BMM_Additive
    rotationSpace=BCS_BoneSpace
    scaleMode=BMM_Additive
    scaleSpace=BCS_BoneSpace
    translationMode=BMM_Additive
    translationSpace=BCS_BoneSpace
    alphaBoolBlend {blendOption=Linear}
    boneToModify {boneName=dyn_skirt_01_r}
    rotation=<rot 5 -20 -10>
   }
   Output Pose(Result=Component To Local(ComponentPose=@n44)) {layerGroup=None}
   n45 = "Constraint - dyn_scarf_l_MSTR (PP)"!(
    ComponentPose=Local To Component(LocalPose=Copy Pose From Mesh() {bUseAttachedParent=true})
    Alpha=1
    ConstraintWeights_0=0.5
    ConstraintWeights_1:float=0.5
   ) -> (Pose) {
    alphaBoolBlend {blendOption=Linear}
    boneToModify {boneName=dyn_scarf_l_MSTR}
    constraintSetup [{targetBone {boneName=spine_05}}, {targetBone {boneName=clavicle_l}}]
    constraintWeights [1,1]
   }
   n46 = "Constraint - dyn_scarf_r_MSTR (PP)"!(ComponentPose=@n45, Alpha=1, ConstraintWeights_0=0.5, ConstraintWeights_1:float=0.5) -> (Pose) {
    alphaBoolBlend {blendOption=Linear}
    boneToModify {boneName=dyn_scarf_r_MSTR}
    constraintSetup [{targetBone {boneName=spine_05}}, {targetBone {boneName=clavicle_r}}]
    constraintWeights [1,1]
   }
   n47 = "Constraint - dyn_bag_02_MSTR (PPPP)"!(
    ComponentPose=@n46
    Alpha=1
    ConstraintWeights_0=0.4605
    ConstraintWeights_1:float=0.1969
    ConstraintWeights_2:float=0.1283
    ConstraintWeights_3:float=0.2144
   ) -> (Pose) {
    alphaBoolBlend {blendOption=Linear}
    boneToModify {boneName=dyn_bag_02_MSTR}
    constraintSetup [{targetBone {boneName=spine_01}}, {targetBone {boneName=spine_02}}, {targetBone {boneName=spine_03}}, {targetBone {boneName=pelvis}}]
    constraintWeights [1,1,1,1]
   }
   n48 = "Constraint - dyn_bag_01_MSTR (PPPP)"!(
    ComponentPose=@n47
    Alpha=1
    ConstraintWeights_0=0.4605
    ConstraintWeights_1:float=0.1969
    ConstraintWeights_2:float=0.1283
    ConstraintWeights_3:float=0.2144
   ) -> (Pose) {
    alphaBoolBlend {blendOption=Linear}
    boneToModify {boneName=dyn_bag_01_MSTR}
    constraintSetup [{targetBone {boneName=spine_01}}, {targetBone {boneName=spine_02}}, {targetBone {boneName=spine_03}}, {targetBone {boneName=pelvis}}]
    constraintWeights [1,1,1,1]
   }
   n49 = "Constraint - dyn_bottle_01_MSTR (PPPP)"!(
    ComponentPose=@n48
    Alpha=1
    ConstraintWeights_0=0.4605
    ConstraintWeights_1:float=0.1969
    ConstraintWeights_2:float=0.1283
    ConstraintWeights_3:float=0.2144
   ) -> (Pose) {
    alphaBoolBlend {blendOption=Linear}
    boneToModify {boneName=dyn_bottle_01_MSTR}
    constraintSetup [{targetBone {boneName=spine_01}}, {targetBone {boneName=spine_02}}, {targetBone {boneName=spine_03}}, {targetBone {boneName=pelvis}}]
    constraintWeights [1,1,1,1]
   }
   n50 = "Constraint - dyn_bottle_02_MSTR (PPPP)"!(
    ComponentPose=@n49
    Alpha=1
    ConstraintWeights_0=0.4605
    ConstraintWeights_1:float=0.1969
    ConstraintWeights_2:float=0.1283
    ConstraintWeights_3:float=0.2144
   ) -> (Pose) {
    alphaBoolBlend {blendOption=Linear}
    boneToModify {boneName=dyn_bottle_02_MSTR}
    constraintSetup [{targetBone {boneName=spine_01}}, {targetBone {boneName=spine_02}}, {targetBone {boneName=spine_03}}, {targetBone {boneName=pelvis}}]
    constraintWeights [1,1,1,1]
   }
   n51 = "Constraint - dyn_bottle_03_MSTR (PPPP)"!(
    ComponentPose=@n50
    Alpha=1
    ConstraintWeights_0=0.3919
    ConstraintWeights_1:float=0.1674
    ConstraintWeights_2:float=0.089
    ConstraintWeights_3:float=0.3517
   ) -> (Pose) {
    alphaBoolBlend {blendOption=Linear}
    boneToModify {boneName=dyn_bottle_03_MSTR}
    constraintSetup [{targetBone {boneName=spine_01}}, {targetBone {boneName=spine_02}}, {targetBone {boneName=spine_03}}, {targetBone {boneName=pelvis}}]
    constraintWeights [1,1,1,1]
   }
   n52 = Get thigh_rot_up_r()
   n53 = "Transform (Modify) Bone - Bone: dyn_bag_02_MSTR"(ComponentPose=@n51, Alpha=@n52, Translation="-1.686,2.271,-1.414", Rotation="7.381,18.857,21.910", Scale="1,1,1") {
    rotationMode=BMM_Additive
    rotationSpace=BCS_BoneSpace
    translationMode=BMM_Additive
    translationSpace=BCS_BoneSpace
    alphaBoolBlend {blendOption=Linear}
    boneToModify {boneName=dyn_bag_02_MSTR}
    rotation=<rot 7.381 18.8571 21.9096>
    translation=<-1.686 2.271 -1.414>
   }
   n54 = "Transform (Modify) Bone - Bone: dyn_bag_01_MSTR"(ComponentPose=@n53, Alpha=@n52, Translation="-1.024,2.307,-0.586", Rotation="-2.374,26.055,32.212", Scale="1,1,1") {
    rotationMode=BMM_Additive
    rotationSpace=BCS_BoneSpace
    translationMode=BMM_Additive
    translationSpace=BCS_BoneSpace
    alphaBoolBlend {blendOption=Linear}
    boneToModify {boneName=dyn_bag_01_MSTR}
    rotation=<rot -2.374 26.0548 32.2122>
    translation=<-1.024 2.307 -0.586>
   }
   n55 = Get thigh_rot_up_l()
   n56 = "Transform (Modify) Bone - Bone: dyn_bottle_03_MSTR"(ComponentPose=@n54, Alpha=@n55, Translation="-6.659,-0.771,-0.000", Rotation="0.000,-5.000,0.000", Scale="0.85,1,1") {
    rotationMode=BMM_Additive
    rotationSpace=BCS_BoneSpace
    scaleMode=BMM_Additive
    scaleSpace=BCS_BoneSpace
    translationMode=BMM_Additive
    translationSpace=BCS_BoneSpace
    alphaBoolBlend {blendOption=Linear}
    boneToModify {boneName=dyn_bottle_03_MSTR}
    rotation {yaw=-5.0001}
    scale {x=0.85}
    translation=<-6.6591 -0.7708>
   }
   n57 = "Transform (Modify) Bone - Bone: dyn_bottle_02_MSTR"(ComponentPose=@n56, Alpha=@n55, Translation="-6.288,-0.527,0.444", Rotation="0.000,-20.001,0.000", Scale="0.85,1,1") {
    rotationMode=BMM_Additive
    rotationSpace=BCS_BoneSpace
    scaleMode=BMM_Additive
    scaleSpace=BCS_BoneSpace
    translationMode=BMM_Additive
    translationSpace=BCS_BoneSpace
    alphaBoolBlend {blendOption=Linear}
    boneToModify {boneName=dyn_bottle_02_MSTR}
    rotation {yaw=-20.001}
    scale {x=0.85}
    translation=<-6.288 -0.527 0.444>
   }
   n58 = "Transform (Modify) Bone - Bone: dyn_bottle_01_MSTR"(ComponentPose=@n57, Alpha=@n55, Translation="-5.941,-0.000,2.778", Rotation="0.000,-25.000,0.000", Scale="0.85,1,1") {
    rotationMode=BMM_Additive
    rotationSpace=BCS_BoneSpace
    scaleMode=BMM_Additive
    scaleSpace=BCS_BoneSpace
    translationMode=BMM_Additive
    translationSpace=BCS_BoneSpace
    alphaBoolBlend {blendOption=Linear}
    boneToModify {boneName=dyn_bottle_01_MSTR}
    rotation {yaw=-24.9996}
    scale {x=0.85}
    translation {x=-5.941, z=2.778}
   }
   n59 = "Transform (Modify) Bone - Bone: dyn_chain_offset_ft"(ComponentPose=@n58, Alpha=Get thigh_rot_up_r(), Translation="-0.000,10.943,11.300", Rotation="19.029,18.643,6.274", Scale="1,1,1") {
    rotationMode=BMM_Additive
    rotationSpace=BCS_BoneSpace
    scaleMode=BMM_Additive
    scaleSpace=BCS_BoneSpace
    translationMode=BMM_Additive
    translationSpace=BCS_BoneSpace
    alphaBoolBlend {blendOption=Linear}
    boneToModify {boneName=dyn_chain_offset_ft}
    rotation=<rot 19.029 18.6431 6.2741>
    translation {y=10.943, z=11.3}
   }
   Save cached pose 'prePose'(Pose=Component To Local(ComponentPose=@n59)) {cacheName=prePose}
   Get thigh_rot_in_r() -> (thigh_rot_in_r:double)
  ]}
 EventGraph {nodes [
   Event BlueprintUpdateAnimation()
   > n1 = Sequence!(execute) -> (then_0, then_1, then_2:Exec)
   n2 = CheckForBackbling()
   > Set isWearingBackbling(isWearingBackbling=@n2)
   > Set Is Backbling A Cape(Is Backbling A Cape:bool=@n2.bIsBackblingACape) -> (Output_Get:bool)
   n3 = Get bRigidBodyEnabled()
   n4 = Get CurrentLOD()
   > Set isRigidBodyDisabled(isRigidBodyDisabled:bool=OR Boolean(A=NOT Boolean(A=@n3), B="integer > integer"(A=@n4, B=1) {is=title K2Node_CallFunction})) -> (Output_Get:bool)
   > n5 = Branch(Condition=Get bIsSwimSprinting() -> (bIsSwimSprinting:bool))
   > Set gravity_override(gravity_override="0, 0, 0")
   n6 = Branch(execute=@n5.else, Condition=Get bIsSkyDiving())
   > n7 = Branch(Condition=Get bIsParachuteOpened())
   Set gravity_override(execute=[@n6.else, @n7.then], gravity_override="0, 0,-980")
   Set gravity_override(execute=@n7.else, gravity_override="0, 0,-300")
   n8 = GetOwnerSkeletalMeshComponent()
   n9 = BreakRotator(InRot=BreakTransform(InTransform=GetDeltaTransformFromRefPose(self=@n8, BoneName=thigh_r, BaseName=pelvis)).Rotation)
   Set thigh_rot_up_r(execute=@n1.then_1, thigh_rot_up_r:double=MapRangeClamped(Value=@n9.Yaw, InRangeA=0, InRangeB=-100)) -> (Output_Get:double)
   > Set thigh_rot_in_r(thigh_rot_in_r:double=MapRangeClamped(Value=@n9.Pitch, InRangeA=0, InRangeB=-15)) -> (Output_Get:double)
   n10 = GetOwnerSkeletalMeshComponent()
   n11 = BreakRotator(InRot=BreakTransform(InTransform=GetDeltaTransformFromRefPose(self=@n10, BoneName=thigh_l, BaseName=pelvis)).Rotation)
   > Set thigh_rot_up_l(thigh_rot_up_l:double=MapRangeClamped(Value=@n11.Yaw, InRangeA=-30, InRangeB=-100)) -> (Output_Get:double)
   > Set thigh_rot_in_l(thigh_rot_in_l:double=MapRangeClamped(Value=@n11.Pitch, InRangeA=0, InRangeB=-15)) -> (Output_Get:double)
   n12 = Branch(execute=@n1.then_2, Condition=Get bIsCrouching())
   > Set crouch(crouch:double=1) -> (Output_Get:double)
   Set crouch(execute=@n12.else, crouch=0)
   n13 = GetDeltaTransformFromRefPose(self=GetOwnerSkeletalMeshComponent(), BoneName=thigh_l, BaseName=thigh_r)
   MapRangeClamped(Value=BreakRotator(InRot=BreakTransform(InTransform=@n13).Rotation).Pitch, InRangeA=0, InRangeB=-40)
  ]}
]
variables [
 crouch {type=number}
 gravity_override {type=Vector}
 Is Backbling A Cape {type=boolean}
 isRigidBodyDisabled {type=boolean}
 isWearingBackbling {type=boolean}
 thigh_rot_in_l {type=number}
 thigh_rot_in_r {type=number}
 thigh_rot_up_l {type=number}
 thigh_rot_up_r {type=number}
]
